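# Niagara system for the BerryFlight skydiving contrail cosmetic, drawing strawberry sprites, colored particles, glitch meshes and hand trails on the character.
# Emitters split particles across two skeletal mesh bones or sockets, using partition-indexed offsets.
# ParaGlideLeanAlpha drives spawn probability, velocity, color and offsets, swapping the two strawberry emitters at a threshold.
# Stepped curves create glitch flicker, while front-end preview and time-of-day inputs adjust placement, scale and brightness.
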NiagaraSystem /BRCosmetics/Athena/Effects/Particles/Skydiving/NS_Contrail_BerryFlight/NS_Contrail_BerryFlight
dependencies {
 usedDataInterfaces [
  &NiagaraDataInterfaceArrayFloat3
  &NiagaraDataInterfaceArrayInt32
  &NiagaraDataInterfaceCurve
  &NiagaraDataInterfaceColorCurve
  &NiagaraDataInterfaceArrayFloat
  /Script/Niagara.NiagaraDataInterfaceArrayQuat
  &NiagaraDataInterfaceSkeletalMesh
 ]
 usedDynamicInputs []
 usedModules [
  &self:SystemSpawnScript.SystemScriptSource.SystemScriptGraph.NiagaraNodeAssignment_0.SetVariables_#1
  &SystemState
  &InterpolateOverTime
 ]
}
emitters {
 Glitch {
  emitterUpdateScript [
   &EmitterState(Life Cycle Mode=::NewEnumerator1, Loop Duration=3, UseLoopDelay=false)
   &SpawnRate(
    Spawn Probability=&Lerp_Float(A=0, B=1, Alpha=&Subtract_Float(A=$User.ParaGlideLeanAlpha, B=0.2))
    SpawnRate=450
    Use Spawn Probability=true
   )
  ]
  particleSpawnScript [
   &PartitionParticles()
   &InitializeParticle(
    Color=&RandomRangeLinearColor(Minimum=<rgba 0.0998 0.9809 1 1.5>, Maximum=<rgba 0.4079 0.1713 0.5 1.5>)
    Lifetime Max=0.3
    Lifetime Min=0.3
    Lifetime Mode=::NewEnumerator1
    Mesh Scale Max=&MakeCustomVectorFromBool(Bool=&RandomBool(), True Vector=<1 1 4>, False Vector=<2.5 2.5 10>)
    Mesh Scale Min=<0.5 0.5 5>
    Mesh Scale Mode=::NewEnumerator2
    UsePositionOffset=true
    Write Material Random=true
   )
   &ShapeLocation(
    Box Size=<20 10 140>
    MaskSpawnGroup=false
    Offset Mode=::NewEnumerator0
    Override Seed=false
    Rotation Mode=::NewEnumerator2
    Shape Origin=$Particles.Position
    Shape Primitive=::NewEnumerator2
    Surface Only=false
    UseOwnerScale=true
    Yaw / Pitch / Roll=&MakeCustomVectorFromBool(
     Bool=$User.bIsFrontEndPreview
     True Vector=&SelectVectorFromArray(
      Direct Array Index=$Particles.PartitionParticles.Partition
      Vector Selection Array=DataInterface() {propertyValues {FloatData [<0 -0.01 0>, <0 0.01 0>]}}
     )
     False Vector=&SelectVectorFromArray(
      Direct Array Index=$Particles.PartitionParticles.Partition
      Vector Selection Array=DataInterface() {propertyValues {FloatData [<0 -0.04 0.03>, <0 0.04 0.03>]}}
     )
    )
   )
   &InitialMeshOrientation(Mesh Orientation Mode=::NewEnumerator2, Rotation=<0.25 0 0>, Use Rotation Vector=true)
   /SetVariables(Particles.VisibilityTag=&MakeCustomIntFromBool(
     Bool=&RandomBool()
     True INT=&RandomRangeInt(Minimum=0, Maximum=1)
     False INT=1
    )) {isSetParameters=true}
  ]
  particleUpdateScript [
   &ParticleState()
   &SkeletalMeshLocation(
    Bone / Socket Index=&SelectIntFromArray(Direct Array Index=$Particles.PartitionParticles.Partition, Int Selection Array=DataInterface() {propertyValues {IntData [0,1]}})
    Bone Sampling Mode=::NewEnumerator4
    MaskSpawnGroup=false
    Orientation Sampling=::NewEnumerator1
    Position Sampling=::NewEnumerator1
    Sampled Position Offset=&AddVectorToPosition(
     Position=&MakeCustomVectorFromBool(
      Bool=$User.bIsFrontEndPreview
      True Vector=&SelectVectorFromArray(
       Direct Array Index=$Particles.PartitionParticles.Partition
       Vector Selection Array=DataInterface() {propertyValues {FloatData [<30 0 85>, <-30 0 85>]}}
      )
      False Vector=&SelectVectorFromArray(
       Direct Array Index=$Particles.PartitionParticles.Partition
       Vector Selection Array=DataInterface() {propertyValues {FloatData [<35 18 85>, <-35 18 85>]}}
      )
     )
     Vector=$Particles.ShapeLocation.ShapeVector
    )
    Sampled Position Offset Coordinate Space:ENiagaraSkelMeshTransforms=::NewEnumerator2
    Skeletal Mesh=$System.Skeletal Mesh Reader
   )
   &ScaleMeshSize(Scale Factor=&Multiply_VectorByFloat(
     Vector=<0.7 0.7 1>
     Float=&Multiply_Float(
      A=&FloatFromCurve(
       FloatCurve=DataInterface() {propertyValues {Curve {keys [<at 0 0 RCIM_Constant>,<at 0.2 1 RCIM_Constant>,<at 0.4 0 RCIM_Constant>,<at 1 1 RCIM_Constant>,<at 1.3 0 RCIM_Constant>,<at 1.9 1 RCIM_Constant>,<at 2.4 0 RCIM_Constant>,<at 3 0 RCIM_Constant>]}}}
       CurveIndex=$Emitter.LoopedAge
      )
      B=&FloatFromCurve(
       FloatCurve=DataInterface() {propertyValues {Curve {keys [
           {arriveTangent=4.1785, time=0, value=1}
           <at 0.5 1.1>
           {arriveTangent=-0.0405, interpMode=RCIM_Cubic, leaveTangent=-0.0405, tangentMode=RCTM_User, time=1, value=1.5}
          ]}}}
       Scale Curve=0.11
      )
     )
    ))
   &Drag(Drag=2, Use Linear Drag=true, Use Rotational Drag=true)
   &SolveForcesAndVelocity(Clamp Velocity=false, Limit Acceleration=false)
   &DynamicMaterialParameters(Module.Param0WriteEnabled=true, Module.Param1WriteEnabled=true, Module.Param2WriteEnabled=true, Module.Param3WriteEnabled=true)
   &ScaleColor(
    Scale Alpha=&Multiply_Float(
     A=&FloatFromCurve(
      FloatCurve=DataInterface() {propertyValues {Curve {keys [
          {arriveTangent=9.0725, interpMode=RCIM_Cubic, leaveTangent=9.0725, tangentMode=RCTM_User, time=0, value=0}
          {arriveTangent=-2.5249, interpMode=RCIM_Cubic, leaveTangent=-2.5249, tangentMode=RCTM_User, time=0.2, value=0.9847}
          {arriveTangent=-0.0405, interpMode=RCIM_Cubic, leaveTangent=-0.0405, tangentMode=RCTM_User, time=1, value=0}
         ]}}}
      Scale Curve=&RandomRangeFloat(Minimum=0.2, Maximum=1)
     )
     B=&FloatFromCurve(
      FloatCurve=DataInterface() {propertyValues {Curve {keys [<at 0 1 RCIM_Constant>,<at 0.05 0.2 RCIM_Constant>,<at 0.12 1 RCIM_Constant>,<at 0.2 0.2 RCIM_Constant>,<at 0.35 1 RCIM_Constant>,<at 0.5 0.2 RCIM_Constant>,<at 0.7 0.2 RCIM_Constant>,<at 0.8 1 RCIM_Constant>,<at 1 0.2 RCIM_Constant>]}}}
      CurveIndex=$Emitter.LoopedAge
     )
    )
    Scale RGB=&Multiply_VectorByFloat(
     Vector=&MakeVectorFromLinearColorRGB(Color=&ColorFromCurve(ColorCurve=DataInterface() {propertyValues {
         AlphaCurve {keys []}
         BlueCurve {keys [
           <at 0 5 RCIM_Cubic>
           <at 0.1326 0>
           <at 0.2869 1.408>
           <at 0.4134 5>
           <at 0.5108 0.2805>
           <at 0.6236 5>
           {arriveTangent=-25.2941, interpMode=RCIM_Cubic, leaveTangent=-25.2941, time=0.7107, value=0}
           <at 0.8097 0.2928>
          ]}
         GreenCurve {keys [
           <at 0 2.5535 RCIM_Cubic>
           <at 0.1326 0>
           <at 0.2869 5>
           <at 0.4134 0>
           <at 0.5108 0>
           <at 0.6236 4.8554>
           {arriveTangent=-26.0903, interpMode=RCIM_Cubic, leaveTangent=-26.0903, time=0.7107, value=0}
           <at 0.8097 0>
          ]}
         RedCurve {keys [
           <at 0 0 RCIM_Cubic>
           <at 0.1326 1>
           <at 0.2869 2.2363>
           <at 0.4134 0.4394>
           <at 0.5108 5>
           <at 0.6236 0>
           {arriveTangent=26.8676, interpMode=RCIM_Cubic, leaveTangent=26.8676, time=0.7107, value=1}
           <at 0.8097 5>
          ]}
        }}))
     Float=&FloatFromCurve(
      FloatCurve=DataInterface() {propertyValues {Curve {keys [{arriveTangent=-2.0684, interpMode=RCIM_Cubic, leaveTangent=-2.0684, tangentMode=RCTM_User, time=0, value=2}, <at 1 0>]}}}
      Scale Curve=&Multiply_Float(A=0.5, B=1)
     )
    )
    ScaleA=true
    ScaleRGB=true
   )
   &FN_Mesh_Attributes(
    Color=&MakeLinearColorFromVectorAndFloat(
     "Vector (RGB)"=&Multiply_Vector(
      A=&MakeVectorFromLinearColorRGB(Color=<rgba 0.4596 0.9809 1 10>)
      B=&Multiply_VectorByFloat(
       Vector=&MakeVectorFromLinearColorRGB(Color=&ColorFromCurve(ColorCurve=DataInterface() {propertyValues {
           AlphaCurve {keys []}
           BlueCurve {keys [
             <at 0 5 RCIM_Cubic>
             <at 0.1326 0>
             <at 0.2869 1.4885>
             <at 0.4134 5>
             <at 0.5507 0.2805>
             <at 0.5846 5>
             <at 0.8175 5>
             {arriveTangent=-29.7003, interpMode=RCIM_Cubic, leaveTangent=-29.7003, time=0.8628, value=0}
             <at 0.976 0.2928>
            ]}
           GreenCurve {keys [
             <at 0 2.5535 RCIM_Cubic>
             <at 0.1326 0>
             <at 0.2869 5>
             <at 0.4134 0>
             <at 0.5507 0>
             <at 0.5846 4.8554>
             <at 0.8175 4.8554>
             {arriveTangent=-30.6352, interpMode=RCIM_Cubic, leaveTangent=-30.6352, time=0.8628, value=0}
             <at 0.976 0>
            ]}
           RedCurve {keys [
             <at 0 0 RCIM_Cubic>
             <at 0.1326 1>
             <at 0.2869 0>
             <at 0.4134 0.4394>
             <at 0.5507 5>
             <at 0.5846 0>
             <at 0.8175 0>
             {arriveTangent=31.5478, interpMode=RCIM_Cubic, leaveTangent=31.5478, time=0.8628, value=1}
             <at 0.976 5>
            ]}
          }}))
       Float=&FloatFromCurve(
        FloatCurve=DataInterface() {propertyValues {Curve {keys [{arriveTangent=-2.0684, interpMode=RCIM_Cubic, leaveTangent=-2.0684, tangentMode=RCTM_User, time=0, value=2}, <at 1 0>]}}}
        Scale Curve=&Multiply_Float(A=1, B=0.5)
       )
      )
     )
     "Float (Alpha)"=&Multiply_Float(
      A=&FloatFromCurve(
       FloatCurve=DataInterface() {propertyValues {Curve {keys [
           {arriveTangent=9.0725, interpMode=RCIM_Cubic, leaveTangent=9.0725, tangentMode=RCTM_User, time=0, value=0}
           {arriveTangent=-2.5249, interpMode=RCIM_Cubic, leaveTangent=-2.5249, tangentMode=RCTM_User, time=0.2, value=0.9847}
           {arriveTangent=-0.0405, interpMode=RCIM_Cubic, leaveTangent=-0.0405, tangentMode=RCTM_User, time=1, value=0}
          ]}}}
       Scale Curve=&RandomRangeFloat(Minimum=0.2, Maximum=1)
      )
      B=2
     )
    )
    Create Color Attribute=true
    Create Scale Attribute=true
    Module Output=::NewEnumerator1
    Scale=&Multiply_Vector(
     A=&RandomRangeVector(
      Minimum=<1 1 10>
      Maximum=&MakeCustomVectorFromBool(Bool=&RandomBool(), True Vector=<1.85 3 1>, False Vector=<1 1 6>)
     )
     B=&Multiply_VectorByFloat(
      Vector=<1 1 1>
      Float=&Multiply_Float(
       A=&MakeCustomFloatFromBool(Bool=&RandomBool(), True Float=0.1, False Float=0)
       B=&FloatFromCurve(
        FloatCurve=DataInterface() {propertyValues {Curve {keys [
            {arriveTangent=4.1785, time=0, value=1}
            <at 0.5 1.1>
            {arriveTangent=-0.0405, interpMode=RCIM_Cubic, leaveTangent=-0.0405, tangentMode=RCTM_User, time=1, value=1.5}
           ]}}}
        Scale Curve=&FloatFromCurve(
         FloatCurve=DataInterface() {propertyValues {Curve {keys [<at 0 0 RCIM_Constant>,<at 0.05 1 RCIM_Constant>,<at 0.3 0 RCIM_Constant>,<at 0.5 1 RCIM_Constant>,<at 0.8 0 RCIM_Constant>,<at 1 1 RCIM_Constant>,<at 1.3 0 RCIM_Constant>,<at 2 0 RCIM_Constant>]}}}
         CurveIndex=$Emitter.LoopedAge
        )
       )
      )
     )
    )
   )
  ]
  rendererClasses [&NiagaraMeshRendererProperties]
  renderers [
   {rendererClass=&NiagaraMeshRendererProperties, rendererIndex=0}
   {rendererClass=&NiagaraMeshRendererProperties, rendererIndex=1}
  ]
 }
 Sprites {
  emitterUpdateScript [
   &EmitterState(
    Enable Distance Culling=true
    Inactive Response=::NewEnumerator1
    Life Cycle Mode=::NewEnumerator1
    Loop Duration=1
    Max Distance Response=::KillImmediately
    Module.bUseMaxDistance=true
    Module.bUseMinDistance=false
    Scalability Mode=::NewEnumerator1
    UseLoopDelay=false
   )
   &SpawnRate(
    Spawn Probability=&MakeCustomFloatFromBool(Bool=&SetBoolByFloatComparison(A=$User.ParaGlideLeanAlpha, B=0.2), True Float=0, False Float=0.3)
    SpawnRate=60
    Use Spawn Probability=true
   )
  ]
  particleSpawnScript [
   &PartitionParticles()
   &InitializeParticle(
    Color=<rgba 0.5 0.5 0.5 1>
    Lifetime Max=0.6
    Lifetime Min=0.4
    Lifetime Mode=::NewEnumerator1
    Sprite Size Mode=::NewEnumerator4
    UsePositionOffset=true
    Write Material Random=true
   )
   &ShapeLocation(MaskSpawnGroup=false, Override Seed=false, Sphere Radius=15, UseOwnerScale=true)
   &AddVelocity(
    Cone Angle=1
    Cone Axis=<0 0 1>
    Module.UseConeFalloff=false
    Module.UseVelDistribution=false
    Override Seed=false
    Rotation Coordinate Space=::World
    Velocity Mode=::NewEnumerator2
    Velocity Speed=&Multiply_Float(
     A=&RandomRangeFloat(Minimum=100, Maximum=200)
     B=&Lerp_Float(
      A=1
      B=&FN_FrontEndFloatModifier(Gameplay=2, Front End=1)
      Alpha=&ClampFloat(Float=$User.ParaGlideLeanAlpha)
     )
    )
   )
   &AddVelocity(Velocity=&MakeCustomVectorFromBool(
     Bool=$User.bIsFrontEndPreview
     True Vector=&SelectVectorFromArray(
      Direct Array Index=$Particles.PartitionParticles.Partition
      Vector Selection Array=DataInterface() {propertyValues {FloatData [<20 0 0>, <-20 0 0>]}}
     )
     False Vector=&SelectVectorFromArray(
      Direct Array Index=$Particles.PartitionParticles.Partition
      Vector Selection Array=DataInterface() {propertyValues {FloatData [<20 40 0>, <-20 40 0>]}}
     )
    )) {module=AddVelocity001}
  ]
  particleUpdateScript [
   &ParticleState()
   &SkeletalMeshLocation(
    Bone / Socket Index=&SelectIntFromArray(Direct Array Index=$Particles.PartitionParticles.Partition, Int Selection Array=DataInterface() {propertyValues {IntData [0,1]}})
    Bone Sampling Mode=::NewEnumerator4
    MaskSpawnGroup=false
    Orientation Sampling=::NewEnumerator1
    Position Sampling=::NewEnumerator1
    Sampled Position Offset=&AddVectorToPosition(
     Position=&MakeCustomVectorFromBool(
      Bool=$User.bIsFrontEndPreview
      True Vector=&SelectVectorFromArray(
       Direct Array Index=$Particles.PartitionParticles.Partition
       Vector Selection Array=DataInterface() {propertyValues {FloatData [<15 0 60>, <-15 0 60>]}}
      )
      False Vector=&SelectVectorFromArray(
       Direct Array Index=$Particles.PartitionParticles.Partition
       Vector Selection Array=DataInterface() {propertyValues {FloatData [<15 10 70>, <-15 10 70>]}}
      )
     )
     Vector=$Particles.ShapeLocation.ShapeVector
    )
    Sampled Position Offset Coordinate Space:ENiagaraSkelMeshTransforms=::NewEnumerator2
    Skeletal Mesh=$System.Skeletal Mesh Reader
   )
   &ScaleSpriteSize(Scale Factor=&Vector2DFromFloat(Value=&FloatFromCurve(FloatCurve=DataInterface() {propertyValues {Curve {keys [<at 0 1>, {arriveTangent=-1.1583, leaveTangent=-1.1583, time=0.5, value=1}, {arriveTangent=-2.2869, time=1, value=0.4}]}}})))
   /SetVariables(Particles.Size002:Vector2f=&Multiply_Vector2D(A=$Particles.SpriteSize, B=&Vector2DFromFloat(Value=1.8))) {isSetParameters=true, module=SetVariables_#2}
   &SolveForcesAndVelocity(Clamp Velocity=false, Limit Acceleration=false)
   &ScaleColor(
    Scale Alpha=&Multiply_Float(
     A=&FNTimeOfDayScalar(Night Scale=1, Morning Scale=0.3, Evening Scale=0.3)
     B=&FloatFromCurve(
      FloatCurve=DataInterface() {propertyValues {Curve {keys [
          {interpMode=RCIM_Cubic, tangentMode=RCTM_User, time=0.1, value=0}
          {interpMode=RCIM_Cubic, tangentMode=RCTM_User, time=0.8, value=1}
          {tangentMode=RCTM_User, time=1, value=0}
         ]}}}
      Scale Curve=&Lerp_Float(A=1, B=3, Alpha=&Subtract_Float(A=$User.ParaGlideLeanAlpha, B=0.2))
     )
    )
    Scale RGB=&Lerp_Vector(
     A=&MakeVectorFromLinearColorRGB(Color=<rgba 1 1 1 1>)
     B=&MakeVectorFromLinearColorRGB(Color=&MakeCustomLinearColorFromBool(
       Bool=&RandomBool()
       True Color=&MakeCustomLinearColorFromBool(
        Bool=&RandomBool()
        True Color=&MakeCustomLinearColorFromBool(Bool=&RandomBool(), True Color=<rgba 1 0.07 0 1>, False Color=<rgba 1 0.8281 0 1>)
        False Color=&MakeCustomLinearColorFromBool(Bool=&RandomBool(), True Color=<rgba 0 1 0.1767 1>, False Color=<rgba 0 1 0.8022 1>)
       )
       False Color=&MakeCustomLinearColorFromBool(
        Bool=&RandomBool()
        True Color=&MakeCustomLinearColorFromBool(Bool=&RandomBool(), True Color=<rgba 0 0.1339 1 1>, False Color=<rgba 0 1 0.8617 1>)
        False Color=&MakeCustomLinearColorFromBool(Bool=&RandomBool(), True Color=<rgba 0.5824 0 1 1>, False Color=<rgba 1 0.8281 0 1>)
       )
      ))
     Alpha=&Subtract_Float(A=$User.ParaGlideLeanAlpha, B=0.2)
    )
    ScaleA=true
    ScaleRGB=true
   )
   /SetVariables(Particles.Color_triangleglow:LinearColor=&MakeLinearColorFromVectorAndFloat(
     "Vector (RGB)"=&MakeVectorFromLinearColorRGB(Color=$Particles.Color)
     "Float (Alpha)"=&Multiply_Float(
      A=&FNTimeOfDayScalar(Night Scale=1, Morning Scale=0.3, Evening Scale=0.3)
      B=&FloatFromCurve(
       FloatCurve=DataInterface() {propertyValues {Curve {keys [
           {interpMode=RCIM_Cubic, tangentMode=RCTM_User, time=0.1, value=0}
           {interpMode=RCIM_Cubic, tangentMode=RCTM_User, time=0.8, value=1}
           {tangentMode=RCTM_User, time=1, value=0}
          ]}}}
       Scale Curve=0.3
      )
     )
    )) {isSetParameters=true, module=SetVariables_#3}
  ]
  rendererClasses [&NiagaraSpriteRendererProperties]
  renderers [
   {rendererClass=&NiagaraSpriteRendererProperties}
   {rendererClass=&NiagaraSpriteRendererProperties, rendererIndex=1}
  ]
 }
 Sprites_Speed {
  emitterUpdateScript [
   &EmitterState(
    Delay First Loop Only=true
    Enable Distance Culling=true
    Inactive Response=::NewEnumerator1
    Life Cycle Mode=::NewEnumerator1
    Loop Delay=1
    Loop Duration=1
    Max Distance Response=::KillImmediately
    Module.bUseMaxDistance=true
    Module.bUseMinDistance=false
    Scalability Mode=::NewEnumerator1
    UseLoopDelay=true
   )
   &SpawnRate(
    Spawn Probability=&Lerp_Float(A=0, B=1, Alpha=&Subtract_Float(A=$User.ParaGlideLeanAlpha, B=0.2))
    SpawnRate=80
    Use Spawn Probability=true
   )
  ]
  particleSpawnScript [
   &PartitionParticles()
   &InitializeParticle(
    Color=<rgba 0.7969 0.7969 0.7969 1>
    Lifetime Max=1
    Lifetime Min=0.7
    Lifetime Mode=::NewEnumerator1
    Sprite Size=<12 12>
    Sprite Size Mode=::NewEnumerator4
    UsePositionOffset=true
    Write Material Random=true
   )
   &ShapeLocation(MaskSpawnGroup=false, Override Seed=false, Sphere Radius=20, UseOwnerScale=true)
   &AddVelocity(
    Cone Angle=6
    Cone Axis=<0 0 1>
    Module.UseConeFalloff=false
    Module.UseVelDistribution=false
    Override Seed=false
    Rotation Coordinate Space=::World
    Velocity Mode=::NewEnumerator2
    Velocity Speed=&Multiply_Float(
     A=&RandomRangeFloat(Minimum=80, Maximum=150)
     B=&Lerp_Float(
      A=1
      B=&FN_FrontEndFloatModifier(Gameplay=2, Front End=1)
      Alpha=&ClampFloat(Float=$User.ParaGlideLeanAlpha)
     )
    )
   )
   &AddVelocity(Velocity=&MakeCustomVectorFromBool(
     Bool=$User.bIsFrontEndPreview
     True Vector=&SelectVectorFromArray(
      Direct Array Index=$Particles.PartitionParticles.Partition
      Vector Selection Array=DataInterface() {propertyValues {FloatData [<20 0 50>, <-20 0 50>]}}
     )
     False Vector=&SelectVectorFromArray(
      Direct Array Index=$Particles.PartitionParticles.Partition
      Vector Selection Array=DataInterface() {propertyValues {FloatData [<50 50 0>, <-50 50 0>]}}
     )
    )) {module=AddVelocity001}
  ]
  particleUpdateScript [
   &ParticleState()
   &SkeletalMeshLocation(
    Bone / Socket Index=&SelectIntFromArray(Direct Array Index=$Particles.PartitionParticles.Partition, Int Selection Array=DataInterface() {propertyValues {IntData [0,1]}})
    Bone Sampling Mode=::NewEnumerator4
    MaskSpawnGroup=false
    Orientation Sampling=::NewEnumerator1
    Position Sampling=::NewEnumerator1
    Sampled Position Offset=&AddVectorToPosition(
     Position=&SelectVectorFromArray(
      Direct Array Index=$Particles.PartitionParticles.Partition
      Vector Selection Array=DataInterface() {propertyValues {FloatData [<20 0 0>, <-20 0 0>]}}
     )
     Vector=$Particles.ShapeLocation.ShapeVector
    )
    Sampled Position Offset Coordinate Space:ENiagaraSkelMeshTransforms=::NewEnumerator2
    Skeletal Mesh=$System.Skeletal Mesh Reader
   )
   &ScaleSpriteSize(Scale Factor=&Vector2DFromFloat(Value=&FloatFromCurve(
      FloatCurve=DataInterface() {propertyValues {Curve {keys [<at 0 0.9>, {arriveTangent=-1.1583, time=0.2, value=1}, <at 0.5 0.875>, <at 0.7 1>, {arriveTangent=-2.2869, time=1, value=0.6}]}}}
      Scale Curve=&RandomRangeFloat(Minimum=0.6, Maximum=1.1)
     )))
   /SetVariables(Particles.Size002:Vector2f=&Multiply_Vector2D(A=$Particles.SpriteSize, B=<1 1>)) {isSetParameters=true, module=SetVariables_#2}
   &SolveForcesAndVelocity(Clamp Velocity=false, Limit Acceleration=false)
   &ScaleColor(
    Scale Alpha=&Multiply_Float(
     A=1
     B=&FloatFromCurve(
      FloatCurve=DataInterface() {propertyValues {Curve {keys [
          {interpMode=RCIM_Cubic, tangentMode=RCTM_User, time=0.1, value=0}
          {arriveTangent=2.1282, interpMode=RCIM_Cubic, leaveTangent=0.0037, tangentMode=RCTM_Break, time=0.6, value=1}
          {interpMode=RCIM_Cubic, tangentMode=RCTM_User, time=0.85, value=1}
          {tangentMode=RCTM_User, time=1, value=0}
         ]}}}
      Scale Curve=&RandomRangeFloat(Minimum=1, Maximum=2)
     )
    )
    Scale RGB=&Lerp_Vector(
     A=&MakeVectorFromLinearColorRGB(Color=<rgba 1 1 1 1>)
     B=&MakeVectorFromLinearColorRGB(Color=&MakeCustomLinearColorFromBool(
       Bool=&RandomBool()
       True Color=&MakeCustomLinearColorFromBool(
        Bool=&RandomBool()
        True Color=&MakeCustomLinearColorFromBool(Bool=&RandomBool(), True Color=<rgba 1 0.1406 0.1935 1>, False Color=<rgba 1 0.8281 0 1>)
        False Color=&MakeCustomLinearColorFromBool(Bool=&RandomBool(), True Color=<rgba 0 1 0.1767 1>, False Color=<rgba 0 1 0.8022 1>)
       )
       False Color=&MakeCustomLinearColorFromBool(
        Bool=&RandomBool()
        True Color=&MakeCustomLinearColorFromBool(Bool=&RandomBool(), True Color=<rgba 0.6719 0.6719 0.6719 1>, False Color=<rgba 1 0.2708 0.3044 1>)
        False Color=&MakeCustomLinearColorFromBool(Bool=&RandomBool(), True Color=<rgba 0.7163 0.1252 1 1>, False Color=<rgba 1 0.348 0.3193 1>)
       )
      ))
     Alpha=&MakeCustomFloatFromBool(Bool=&SetBoolByFloatComparison(A=$User.ParaGlideLeanAlpha, B=0), True Float=1, False Float=0)
    )
    ScaleA=true
    ScaleRGB=true
   )
   /SetVariables(Particles.Color_triangleglow:LinearColor=&MakeLinearColorFromVectorAndFloat(
     "Vector (RGB)"=&MakeVectorFromLinearColorRGB(Color=$Particles.Color)
     "Float (Alpha)"=&Multiply_Float(
      A=&FNTimeOfDayScalar(Night Scale=1, Morning Scale=0.3, Evening Scale=0.3)
      B=&FloatFromCurve(
       FloatCurve=DataInterface() {propertyValues {Curve {keys [
           {interpMode=RCIM_Cubic, tangentMode=RCTM_User, time=0.1, value=0}
           {arriveTangent=2.2767, interpMode=RCIM_Cubic, leaveTangent=0.0037, tangentMode=RCTM_Break, time=0.6, value=1}
           {interpMode=RCIM_Cubic, tangentMode=RCTM_User, time=0.85, value=1}
           {tangentMode=RCTM_User, time=1, value=0}
          ]}}}
       Scale Curve=0.5
      )
     )
    )) {isSetParameters=true, module=SetVariables_#3}
  ]
  rendererClasses [&NiagaraSpriteRendererProperties]
  renderers [
   {rendererClass=&NiagaraSpriteRendererProperties}
   {rendererClass=&NiagaraSpriteRendererProperties, rendererIndex=1}
  ]
 }
 Strawberry {
  emitterUpdateScript [
   &EmitterState(
    Inactive Response=::NewEnumerator1
    Life Cycle Mode=::NewEnumerator1
    Loop Behavior=::NewEnumerator1
    Loop Duration Mode=::NewEnumerator1
    UseLoopDelay=false
   )
   &SpawnBurst_Instantaneous(Spawn Count=2, Use Spawn Probability=false, UseLoopCountLimit=false)
  ]
  particleSpawnScript [
   &PartitionParticles()
   &InitializeParticle(
    Color Mode=::NewEnumerator0
    Lifetime=5
    Sprite Size Mode=::NewEnumerator3
    Uniform Sprite Size=50
    UsePositionOffset=false
    Write Material Random=true
   )
  ]
  particleUpdateScript [
   &ParticleState(Kill Particles When Lifetime Has Elapsed=false, Let Infinitely Lived Particles Die When Emitter Deactivates=true)
   /SetVariables(Particles.SpriteRotation:NiagaraFloat=&FN_FrontEndFloatModifier(
     Gameplay=&SelectFloatFromArray(Direct Array Index=$Particles.PartitionParticles.Partition, Float Selection Array=DataInterface() {propertyValues {FloatData [-10,10]}})
     Front End=0
    )) {isSetParameters=true}
   &ScaleSpriteSize(
    Uniform Curve Scale=&MakeCustomFloatFromBool(Bool=&SetBoolByFloatComparison(A=$User.ParaGlideLeanAlpha, B=0.2), True Float=0, False Float=1)
    Uniform Curve Sprite Scale=DataInterface() {propertyValues {Curve {keys [{tangentMode=RCTM_User, time=0, value=1}, {tangentMode=RCTM_User, time=1, value=1}]}}}
   )
   &ScaleSpriteSize(
    Scale Sprite Size Mode=::NewEnumerator0
    Uniform Scale Factor=&FN_FrontEndFloatModifier(Gameplay=1, Front End=0.85)
   ) {module=ScaleSpriteSize001}
   &SkeletalMeshLocation(
    Bone / Socket Index=$Particles.PartitionParticles.Partition
    Bone Sampling Mode=::NewEnumerator4
    MaskSpawnGroup=false
    Orientation Sampling=::NewEnumerator1
    Sampled Position Offset=&FN_FrontEndVector3Modifier(
     Gameplay=&Lerp_Vector(
      A=&SelectVectorFromArray(
       Direct Array Index=$Particles.PartitionParticles.Partition
       Vector Selection Array=DataInterface() {propertyValues {FloatData [<8 0 -25>, <-10 0 -25>]}}
      )
      B=&SelectVectorFromArray(
       Direct Array Index=$Particles.PartitionParticles.Partition
       Vector Selection Array=DataInterface() {propertyValues {FloatData [<20 0 0>, <-20 0 0>]}}
      )
      Alpha=$User.ParaGlideLeanAlpha
     )
     Front End=&Lerp_Vector(
      A=&SelectVectorFromArray(
       Direct Array Index=$Particles.PartitionParticles.Partition
       Vector Selection Array=DataInterface() {propertyValues {FloatData [<8 0 -20>, <-10 0 -20>]}}
      )
      B=&SelectVectorFromArray(
       Direct Array Index=$Particles.PartitionParticles.Partition
       Vector Selection Array=DataInterface() {propertyValues {FloatData [<20 0 0>, <-20 0 0>]}}
      )
      Alpha=$User.ParaGlideLeanAlpha
     )
    )
    Sampled Position Offset Coordinate Space:ENiagaraSkelMeshTransforms=::NewEnumerator2
    Skeletal Mesh=$System.Skeletal Mesh Reader
   )
   &ScaleColor(Scale Alpha=5, ScaleA=true, ScaleRGB=true)
   /SetVariables(Particles.Color_Scanline:LinearColor=&MakeLinearColorFromVectorAndFloat("Vector (RGB)"=&MakeVectorFromLinearColorRGB(Color=<rgba 0.18 0.18 0.18 1>), "Float (Alpha)"=10)) {isSetParameters=true, module=SetVariables_#4}
   /SetVariables(Particles.Color_Glow:LinearColor=&MakeLinearColorFromVectorAndFloat(
     "Vector (RGB)"=&MakeVectorFromLinearColorRGB(Color=<rgba 3 3 3 1>)
     "Float (Alpha)"=&FN_FrontEndFloatModifier(Gameplay=4, Front End=2.5)
    )) {isSetParameters=true, module=SetVariables_#5}
  ]
  rendererClasses [&NiagaraSpriteRendererProperties]
  renderers [
   {rendererClass=&NiagaraSpriteRendererProperties}
   {rendererClass=&NiagaraSpriteRendererProperties, rendererIndex=1}
   {rendererClass=&NiagaraSpriteRendererProperties, rendererIndex=2}
  ]
 }
 Strawberry_Speed001 {
  emitterUpdateScript [
   &EmitterState(
    Inactive Response=::NewEnumerator1
    Life Cycle Mode=::NewEnumerator1
    Loop Behavior=::NewEnumerator1
    Loop Duration Mode=::NewEnumerator1
    UseLoopDelay=false
   )
   &SpawnBurst_Instantaneous(Spawn Count=2, Use Spawn Probability=false, UseLoopCountLimit=false)
  ]
  particleSpawnScript [
   &PartitionParticles()
   &InitializeParticle(
    Color Mode=::NewEnumerator0
    Lifetime=5
    Sprite Rotation Angle=&SelectFloatFromArray(Direct Array Index=$Particles.PartitionParticles.Partition, Float Selection Array=DataInterface() {propertyValues {FloatData [-0.02,0.02]}})
    Sprite Rotation Mode=::NewEnumerator1
    Sprite Size Mode=::NewEnumerator3
    Uniform Sprite Size=50
    UsePositionOffset=false
    Write Material Random=true
   )
  ]
  particleUpdateScript [
   &ParticleState(Kill Particles When Lifetime Has Elapsed=false, Let Infinitely Lived Particles Die When Emitter Deactivates=true)
   /SetVariables(Particles.SpriteRotation:NiagaraFloat=&FN_FrontEndFloatModifier(
     Gameplay=&SelectFloatFromArray(Direct Array Index=$Particles.PartitionParticles.Partition, Float Selection Array=DataInterface() {propertyValues {FloatData [-10,10]}})
     Front End=0
    )) {isSetParameters=true}
   &ScaleSpriteSize(
    Uniform Curve Scale=&MakeCustomFloatFromBool(Bool=&SetBoolByFloatComparison(A=$User.ParaGlideLeanAlpha, B=0.2), True Float=1, False Float=0)
    Uniform Curve Sprite Scale=DataInterface() {propertyValues {Curve {keys [{tangentMode=RCTM_User, time=0, value=1}, {tangentMode=RCTM_User, time=1, value=1}]}}}
   )
   &ScaleSpriteSize(
    Scale Sprite Size Mode=::NewEnumerator0
    Uniform Scale Factor=&FN_FrontEndFloatModifier(Gameplay=1, Front End=0.85)
   ) {module=ScaleSpriteSize001}
   &SkeletalMeshLocation(
    Bone / Socket Index=$Particles.PartitionParticles.Partition
    Bone Sampling Mode=::NewEnumerator4
    MaskSpawnGroup=false
    Orientation Sampling=::NewEnumerator1
    Sampled Position Offset=&FN_FrontEndVector3Modifier(
     Gameplay=&Lerp_Vector(
      A=&SelectVectorFromArray(
       Direct Array Index=$Particles.PartitionParticles.Partition
       Vector Selection Array=DataInterface() {propertyValues {FloatData [<8 0 -25>, <-10 0 -25>]}}
      )
      B=&SelectVectorFromArray(
       Direct Array Index=$Particles.PartitionParticles.Partition
       Vector Selection Array=DataInterface() {propertyValues {FloatData [<20 0 0>, <-20 0 0>]}}
      )
      Alpha=$User.ParaGlideLeanAlpha
     )
     Front End=&Lerp_Vector(
      A=&SelectVectorFromArray(
       Direct Array Index=$Particles.PartitionParticles.Partition
       Vector Selection Array=DataInterface() {propertyValues {FloatData [<8 0 -20>, <-10 0 -20>]}}
      )
      B=&SelectVectorFromArray(
       Direct Array Index=$Particles.PartitionParticles.Partition
       Vector Selection Array=DataInterface() {propertyValues {FloatData [<20 0 0>, <-20 0 0>]}}
      )
      Alpha=$User.ParaGlideLeanAlpha
     )
    )
    Sampled Position Offset Coordinate Space:ENiagaraSkelMeshTransforms=::NewEnumerator2
    Skeletal Mesh=$System.Skeletal Mesh Reader
   )
   &ScaleColor(Scale Alpha=5, ScaleA=true, ScaleRGB=true)
   /SetVariables(Particles.Color_Scanline:LinearColor=&MakeLinearColorFromVectorAndFloat("Vector (RGB)"=&MakeVectorFromLinearColorRGB(Color=<rgba 0.18 0.18 0.18 1>), "Float (Alpha)"=10)) {isSetParameters=true, module=SetVariables_#4}
   /SetVariables(Particles.Color_Glow:LinearColor=&MakeLinearColorFromVectorAndFloat(
     "Vector (RGB)"=&MakeVectorFromLinearColorRGB(Color=<rgba 3 3 3 1>)
     "Float (Alpha)"=&FN_FrontEndFloatModifier(Gameplay=4, Front End=2.5)
    )) {isSetParameters=true, module=SetVariables_#5}
  ]
  rendererClasses [&NiagaraSpriteRendererProperties]
  renderers [
   {rendererClass=&NiagaraSpriteRendererProperties}
   {rendererClass=&NiagaraSpriteRendererProperties, rendererIndex=1}
   {rendererClass=&NiagaraSpriteRendererProperties, rendererIndex=2}
  ]
 }
 Trail_Hand002 {
  emitterUpdateScript [
   &EmitterState(
    Inactive Response=::NewEnumerator1
    Life Cycle Mode=::NewEnumerator1
    Loop Behavior=::NewEnumerator1
    Loop Duration Mode=::NewEnumerator1
    UseLoopDelay=false
   )
   &SpawnBurst_Instantaneous(Spawn Count=2, Use Spawn Probability=false, UseLoopCountLimit=false)
   /SetVariables(Emitter.TrailTex_001U:NiagaraFloat=&MakeCustomFloatFromBool(Bool=&SetBoolByFloatComparison(A=$User.ParaGlideLeanAlpha, B=0.2), True Float=1.7, False Float=1.4)) {isSetParameters=true}
   /SetVariables(Emitter.TrailTex_001SpeedU:NiagaraFloat=&MakeCustomFloatFromBool(Bool=&SetBoolByFloatComparison(A=$User.ParaGlideLeanAlpha, B=0.2), True Float=-2.5, False Float=-1.8)) {isSetParameters=true}
  ]
  particleSpawnScript [
   &PartitionParticles()
   &InitializeParticle(
    Color Mode=::NewEnumerator0
    Lifetime=3
    Mesh Scale=<0.6 1 1>
    Mesh Scale Mode=::NewEnumerator4
    UsePositionOffset=false
    Write Material Random=true
   )
  ]
  particleUpdateScript [
   &ParticleState(Kill Particles When Lifetime Has Elapsed=false, Loop Particles Lifetime=true)
   &ScaleMeshSize(Scale Factor=&MakeCustomVectorFromBool(
     Bool=$User.bisFrontEndPreview
     True Vector=&Lerp_Vector(A=<0.9 1 1>, B=<1.2 1 1>, Alpha=$System.InterpolateOverTime.InterpolateOverTime_Float.Moving Average)
     False Vector=&Lerp_Vector(A=<1.1 1 1>, B=<1.4 1 1>, Alpha=$System.InterpolateOverTime.InterpolateOverTime_Float.Moving Average)
    ))
   /SetVariables(Particles.Scale_Rainbow:Vector3f=&MakeCustomVectorFromBool(
     Bool=$User.bisFrontEndPreview
     True Vector=&Lerp_Vector(A=<0.6 1 1>, B=<1 1 1>, Alpha=$System.InterpolateOverTime.InterpolateOverTime_Float.Moving Average)
     False Vector=&Lerp_Vector(A=<0.8 1 1>, B=<1 1 1>, Alpha=$System.InterpolateOverTime.InterpolateOverTime_Float.Moving Average)
    )) {isSetParameters=true}
   /SetVariables(Particles.Scale_glow:Vector3f=&MakeCustomVectorFromBool(
     Bool=$User.bisFrontEndPreview
     True Vector=&Lerp_Vector(A=<0.6 1 1>, B=<1 1 1>, Alpha=$System.InterpolateOverTime.InterpolateOverTime_Float.Moving Average)
     False Vector=&Lerp_Vector(A=<0.5 1 1>, B=<1.3 1 1>, Alpha=$System.InterpolateOverTime.InterpolateOverTime_Float.Moving Average)
    )) {isSetParameters=true}
   &SkeletalMeshLocation(
    Bone / Socket Index=&SelectIntFromArray(Direct Array Index=$Particles.PartitionParticles.Partition, Int Selection Array=DataInterface() {propertyValues {IntData [0,1]}})
    Bone Sampling Mode=::NewEnumerator4
    MaskSpawnGroup=false
    Orientation Sampling=::NewEnumerator1
    Sampled Position Offset=&FN_FrontEndVector3Modifier(
     Gameplay=&Lerp_Vector(
      A=&SelectVectorFromArray(
       Direct Array Index=$Particles.PartitionParticles.Partition
       Vector Selection Array=DataInterface() {propertyValues {FloatData [<10 0 -10>, <-10 0 -10>]}}
      )
      B=&SelectVectorFromArray(
       Direct Array Index=$Particles.PartitionParticles.Partition
       Vector Selection Array=DataInterface() {propertyValues {FloatData [<25 0 5>, <-20 0 5>]}}
      )
      Alpha=$User.ParaGlideLeanAlpha
     )
     Front End=&Lerp_Vector(
      A=&SelectVectorFromArray(
       Direct Array Index=$Particles.PartitionParticles.Partition
       Vector Selection Array=DataInterface() {propertyValues {FloatData [<10 0 -10>, <-10 0 -10>]}}
      )
      B=&SelectVectorFromArray(
       Direct Array Index=$Particles.PartitionParticles.Partition
       Vector Selection Array=DataInterface() {propertyValues {FloatData [<25 0 5>, <-20 0 5>]}}
      )
      Alpha=$User.ParaGlideLeanAlpha
     )
    )
    Sampled Position Offset Coordinate Space:ENiagaraSkelMeshTransforms=::NewEnumerator2
    Skeletal Mesh=$System.Skeletal Mesh Reader
   )
   &ScaleColor(
    Scale Alpha=&Multiply_Float(
     A=25
     B=&Lerp_Float(
      A=1
      B=&FloatFromCurve(FloatCurve=DataInterface() {propertyValues {Curve {keys [<at 0 1 RCIM_Constant>,<at 0.05 0.2 RCIM_Constant>,<at 0.12 1 RCIM_Constant>,<at 0.3 0.2 RCIM_Constant>,<at 0.4 1 RCIM_Constant>,<at 0.6 0.2 RCIM_Constant>,<at 0.7 0.2 RCIM_Constant>,<at 0.75 1 RCIM_Constant>,<at 1 0.2 RCIM_Constant>]}}})
      Alpha=&Subtract_Float(A=$User.ParaGlideLeanAlpha, B=0.2)
     )
    )
    Scale RGB=&MakeVectorFromLinearColorRGB(Color=&MakeCustomLinearColorFromBool(Bool=$User.bIsFrontEndPreview, True Color=<rgba 0.3 0.3 0.3 1>, False Color=<rgba 0.5 0.5 0.5 1>))
    ScaleA=true
    ScaleRGB=true
   )
   /SetVariables(Particles.ColorPoint:LinearColor=&MakeLinearColorFromVectorAndFloat(
     "Vector (RGB)"=&Multiply_VectorByFloat(Vector=&MakeVectorFromLinearColorRGB(Color=<rgba 5 5 5 1>), Float=1.5)
     "Float (Alpha)"=&Multiply_Float(
      A=&FN_FrontEndFloatModifier(Gameplay=10, Front End=3)
      B=&Lerp_Float(
       A=1
       B=&FloatFromCurve(FloatCurve=DataInterface() {propertyValues {Curve {keys [<at 0 1 RCIM_Constant>,<at 0.05 0 RCIM_Constant>,<at 0.12 1 RCIM_Constant>,<at 0.3 0 RCIM_Constant>,<at 0.4 1 RCIM_Constant>,<at 0.6 0 RCIM_Constant>,<at 0.7 0 RCIM_Constant>,<at 0.75 1 RCIM_Constant>,<at 1 0 RCIM_Constant>]}}})
       Alpha=&Subtract_Float(A=$User.ParaGlideLeanAlpha, B=0.2)
      )
     )
    )) {isSetParameters=true}
   /SetVariables(Particles.Color_Bottom02:LinearColor=&MakeLinearColorFromVectorAndFloat(
     "Vector (RGB)"=&MakeVectorFromLinearColorRGB(Color=<rgba 0.5781 0.5781 0.5781 1>)
     "Float (Alpha)"=&Multiply_Float(
      A=30
      B=&Lerp_Float(
       A=1
       B=&FloatFromCurve(FloatCurve=DataInterface() {propertyValues {Curve {keys [<at 0 1 RCIM_Constant>,<at 0.05 0 RCIM_Constant>,<at 0.12 1 RCIM_Constant>,<at 0.3 0 RCIM_Constant>,<at 0.4 1 RCIM_Constant>,<at 0.6 0 RCIM_Constant>,<at 0.7 0 RCIM_Constant>,<at 0.75 1 RCIM_Constant>,<at 1 0 RCIM_Constant>]}}})
       Alpha=&Subtract_Float(A=$User.ParaGlideLeanAlpha, B=0.2)
      )
     )
    )) {isSetParameters=true}
   /SetVariables(Particles.ColorGlow:LinearColor=&MakeLinearColorFromVectorAndFloat(
     "Vector (RGB)"=&MakeVectorFromLinearColorRGB(Color=<rgba 0.8 0.8 0.8 1>)
     "Float (Alpha)"=&Multiply_Float(
      A=&FN_FrontEndFloatModifier(Gameplay=1, Front End=0.5)
      B=&Lerp_Float(
       A=1
       B=&FloatFromCurve(FloatCurve=DataInterface() {propertyValues {Curve {keys [<at 0 1 RCIM_Constant>,<at 0.05 0 RCIM_Constant>,<at 0.12 1 RCIM_Constant>,<at 0.3 0 RCIM_Constant>,<at 0.4 1 RCIM_Constant>,<at 0.6 0 RCIM_Constant>,<at 0.7 0 RCIM_Constant>,<at 0.75 1 RCIM_Constant>,<at 1 0 RCIM_Constant>]}}})
       Alpha=&Subtract_Float(A=$User.ParaGlideLeanAlpha, B=0.2)
      )
     )
    )) {isSetParameters=true}
   /SetVariables(Particles.Color_Rainbow:LinearColor=&MakeLinearColorFromVectorAndFloat(
     "Vector (RGB)"=&MakeVectorFromLinearColorRGB(Color=&ColorFromCurve(ColorCurve=DataInterface() {propertyValues {
         AlphaCurve {keys [<at 0 1 RCIM_Cubic>,<at 1 0 RCIM_Cubic>]}
         BlueCurve {keys [<at 0 1 RCIM_Cubic>,<at 1 1 RCIM_Cubic>]}
         GreenCurve {keys [<at 0 1 RCIM_Cubic>,<at 1 1 RCIM_Cubic>]}
         RedCurve {keys [<at 0 1 RCIM_Cubic>,<at 1 1 RCIM_Cubic>]}
        }}))
     "Float (Alpha)"=&Multiply_Float(
      A=&Multiply_Float(A=30, B=&FNTimeOfDayScalar(Night Scale=0.65, Morning Scale=0.85, Evening Scale=0.85))
      B=&Lerp_Float(
       A=0
       B=&FloatFromCurve(FloatCurve=DataInterface() {propertyValues {Curve {keys [<at 0 0 RCIM_Constant>,<at 0.05 1 RCIM_Constant>,<at 0.12 0 RCIM_Constant>,<at 0.3 1 RCIM_Constant>,<at 0.4 0 RCIM_Constant>,<at 0.6 1 RCIM_Constant>,<at 0.7 1 RCIM_Constant>,<at 0.75 0 RCIM_Constant>,<at 0.9 0 RCIM_Constant>,<at 1 0 RCIM_Constant>]}}})
       Alpha=&Subtract_Float(A=$User.ParaGlideLeanAlpha, B=0.2)
      )
     )
    )) {isSetParameters=true}
   /SetVariables(Particles.Color_Rainbow_Out:LinearColor=&MakeLinearColorFromVectorAndFloat(
     "Vector (RGB)"=&MakeVectorFromLinearColorRGB(Color=<rgba 0.5781 0.0636 0.0452 1>)
     "Float (Alpha)"=&Multiply_Float(
      A=100
      B=&Lerp_Float(
       A=0
       B=&FloatFromCurve(FloatCurve=DataInterface() {propertyValues {Curve {keys [<at 0 0 RCIM_Constant>,<at 0.05 1 RCIM_Constant>,<at 0.12 0 RCIM_Constant>,<at 0.3 1 RCIM_Constant>,<at 0.4 0 RCIM_Constant>,<at 0.6 1 RCIM_Constant>,<at 0.7 1 RCIM_Constant>,<at 0.75 0 RCIM_Constant>,<at 0.9 0 RCIM_Constant>,<at 1 0 RCIM_Constant>]}}})
       Alpha=&Subtract_Float(A=$User.ParaGlideLeanAlpha, B=0.2)
      )
     )
    )) {isSetParameters=true}
   &DynamicMaterialParameters(
    Index 2 Param 1=1
    Index 2 Param 3=&MakeCustomFloatFromBool(Bool=$User.bisFrontEndPreview, True Float=0.7, False Float=1)
    Index 3 Param 1=&FloatFromCurve(FloatCurve=DataInterface() {propertyValues {Curve {keys [<at 0 0 RCIM_Constant>,<at 0.05 1 RCIM_Constant>,<at 0.12 0 RCIM_Constant>,<at 0.3 1 RCIM_Constant>,<at 0.4 0 RCIM_Constant>,<at 0.6 1 RCIM_Constant>,<at 0.7 1 RCIM_Constant>,<at 0.75 0 RCIM_Constant>,<at 0.9 0 RCIM_Constant>,<at 1 0 RCIM_Constant>]}}})
    Module.Index2Float0Write=true
    Module.Index2Float1Write=true
    Module.Index2Float2Write=true
    Module.Index2Float3Write=false
    Module.Index3Float0Write=true
    Module.Index3Float1Write=false
    Module.Index3Float2Write=false
    Module.Index3Float3Write=false
    Write Parameter Index 0=::NewEnumerator0
    Write Parameter Index 2=::NewEnumerator1
    Write Parameter Index 3=::NewEnumerator1
   )
   /SetVariables(Particles.MeshOrientation=/Niagara/DynamicInputs/TypeConversions/MakeCustomQuaternionFromBool(
     Bool:NiagaraBool=$User.bIsFrontEndPreview
     True Quat:Quat4f=&LerpQuaternion(
      Quaternion A:Quat4f=/Niagara/DynamicInputs/Arrays/SelectQuaternionFromArray(
       Array Sampling Mode:ENiagaraArraySamplingMode=::NewEnumerator1
       Direct Array Index:NiagaraInt32=$Particles.PartitionParticles.Partition
       Interpolation:NiagaraFloat=?
       Direct Array Mode:ENiagara_WrapClamp=::NewEnumerator0
       Evaluation Type:ENiagaraRandomnessEvaluation=::NewEnumerator0
       Normalized Interpolation:NiagaraBool=true
       Recalculate Random Each Loop:NiagaraBool=false
       Randomness Mode:ENiagaraRandomnessMode=::NewEnumerator0
       Random Seed:NiagaraInt32=?
       Fixed Random Seed:NiagaraBool=false
       Quaternion Selection Array:NiagaraDataInterfaceArrayQuat=DataInterface() {propertyValues {GpuSyncMode=SyncCpuToGpu, MaxElements=0, QuatData [<0.1173 -0.6653 0.128 0.7261>, <0.6653 -0.1173 0.7261 0.128>]}}
       Enable Spherical Linear Interpolation:NiagaraBool=?
      )
      Quaternion B:Quat4f=&SelectQuaternionFromArray(
       Array Sampling Mode:ENiagaraArraySamplingMode=::NewEnumerator1
       Direct Array Index:NiagaraInt32=$Particles.PartitionParticles.Partition
       Interpolation:NiagaraFloat=?
       Direct Array Mode:ENiagara_WrapClamp=::NewEnumerator0
       Evaluation Type:ENiagaraRandomnessEvaluation=::NewEnumerator0
       Normalized Interpolation:NiagaraBool=true
       Recalculate Random Each Loop:NiagaraBool=false
       Randomness Mode:ENiagaraRandomnessMode=::NewEnumerator0
       Random Seed:NiagaraInt32=?
       Fixed Random Seed:NiagaraBool=false
       Quaternion Selection Array:NiagaraDataInterfaceArrayQuat=DataInterface() {propertyValues {GpuSyncMode=SyncCpuToGpu, MaxElements=0, QuatData [<0.1173 -0.6653 0.128 0.7261>, <0.128 0.7261 0.1173 -0.6653>]}}
       Enable Spherical Linear Interpolation:NiagaraBool=?
      )
      Lerp Factor:NiagaraFloat=&Subtract_Float(A=$User.ParaGlideLeanAlpha, B=0.2)
      Enable Spherical Linear Interpolation:NiagaraBool=false
     )
     False Quat:Quat4f=&LerpQuaternion(
      Quaternion A:Quat4f=&SelectQuaternionFromArray(
       Array Sampling Mode:ENiagaraArraySamplingMode=::NewEnumerator1
       Direct Array Index:NiagaraInt32=$Particles.PartitionParticles.Partition
       Interpolation:NiagaraFloat=?
       Direct Array Mode:ENiagara_WrapClamp=::NewEnumerator0
       Evaluation Type:ENiagaraRandomnessEvaluation=::NewEnumerator0
       Normalized Interpolation:NiagaraBool=true
       Recalculate Random Each Loop:NiagaraBool=false
       Randomness Mode:ENiagaraRandomnessMode=::NewEnumerator0
       Random Seed:NiagaraInt32=?
       Fixed Random Seed:NiagaraBool=false
       Quaternion Selection Array:NiagaraDataInterfaceArrayQuat=DataInterface() {propertyValues {GpuSyncMode=SyncCpuToGpu, MaxElements=0, QuatData [<0.2717 -0.5826 0.3237 0.6943>, <0.5826 -0.2717 0.6943 0.3237>]}}
       Enable Spherical Linear Interpolation:NiagaraBool=?
      )
      Quaternion B:Quat4f=&SelectQuaternionFromArray(
       Array Sampling Mode:ENiagaraArraySamplingMode=::NewEnumerator1
       Direct Array Index:NiagaraInt32=$Particles.PartitionParticles.Partition
       Interpolation:NiagaraFloat=?
       Direct Array Mode:ENiagara_WrapClamp=::NewEnumerator0
       Evaluation Type:ENiagaraRandomnessEvaluation=::NewEnumerator0
       Normalized Interpolation:NiagaraBool=true
       Recalculate Random Each Loop:NiagaraBool=false
       Randomness Mode:ENiagaraRandomnessMode=::NewEnumerator0
       Random Seed:NiagaraInt32=?
       Fixed Random Seed:NiagaraBool=false
       Quaternion Selection Array:NiagaraDataInterfaceArrayQuat=DataInterface() {propertyValues {GpuSyncMode=SyncCpuToGpu, MaxElements=0, QuatData [<0.2424 -0.5198 0.3462 0.7424>, <-0.3135 -0.7568 -0.2195 0.5299>]}}
       Enable Spherical Linear Interpolation:NiagaraBool=?
      )
      Lerp Factor:NiagaraFloat=&Subtract_Float(A=$User.ParaGlideLeanAlpha, B=0.2)
      Enable Spherical Linear Interpolation:NiagaraBool=false
     )
    )) {isSetParameters=true}
  ]
  rendererClasses [&NiagaraMeshRendererProperties]
  renderers [
   {rendererClass=&NiagaraMeshRendererProperties, rendererIndex=0}
   {rendererClass=&NiagaraMeshRendererProperties, rendererIndex=1}
   {rendererClass=&NiagaraMeshRendererProperties, rendererIndex=2}
   {rendererClass=&NiagaraMeshRendererProperties, rendererIndex=3}
   {rendererClass=&NiagaraMeshRendererProperties, rendererIndex=4}
   {rendererClass=&NiagaraMeshRendererProperties, rendererIndex=5}
  ]
 }
}
summary {
 systemName=NS_Contrail_BerryFlight
 emitters [
  {emitterName=Sprites, rendererClasses [&NiagaraSpriteRendererProperties]}
  {emitterName=Sprites_Speed, rendererClasses [&NiagaraSpriteRendererProperties]}
  {emitterName=Glitch, rendererClasses [&NiagaraMeshRendererProperties]}
  {emitterName=Strawberry, rendererClasses [&NiagaraSpriteRendererProperties]}
  {emitterName=Strawberry_Speed001, rendererClasses [&NiagaraSpriteRendererProperties]}
  {emitterName=Trail_Hand002, rendererClasses [&NiagaraMeshRendererProperties]}
 ]
 userVariables [
  User.bIsFrontEnd {type=NiagaraBool}
  User.bIsFrontEndPreview {type=NiagaraBool}
  User.ParaGlideLeanAlpha {type=NiagaraFloat, defaultValue=1, description=""}
 ]
}
